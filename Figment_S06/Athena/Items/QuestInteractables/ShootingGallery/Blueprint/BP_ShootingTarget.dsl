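; Flip-up target for the shooting gallery quest, activated per round by board ID and reporting results to its GamePad gallery through ShootingGalleryIF.
; The server times each board's live window and accepts only non-pickaxe panel hits while it is ready and the gallery active.
; RepNotifies drive the rest: bAlive plays open/close timelines and sounds, bReady snaps server panel rotation, bScored reports hits.
; Forced deactivation closes the board quickly; damage also resets its building health.
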(event EventActivateBoard
  (params (BoardID :type Integer) (CurrentRound :type Integer))
  ;[0.2005 0.1189 1] Activation
  (Variables|Default|SetMyBoardID BoardID)
  (Variables|Default|SetbScored false)
  (Variables|Default|SetCurrentRound CurrentRound)
  (CallFunction|ClearDeactivate)
  (Variables|Default|SetbAlive true)
  (Networking|SwitchHasAuthority
    (:Authority
      (Utilities|FlowControl|Sequence
        (:then_0
          ;[1 0.3126 0] Server tracking active state of board
          (Utilities|FlowControl|Delay 0.5)
          (Variables|Default|SetbReady true)
          (Utilities|FlowControl|Delay (Utilities|Array|Get(aref) (Variables|GameBoard|GetTimeAlive) (Variables|Default|GetCurrentRound)))
          (if (Variables|Default|GetbAlive)
            _))
        (:then_1
          (Utilities|FlowControl|DoOnce false
            (:Completed
              _))))
      (Variables|Default|SetbAlive false)
      (Variables|Default|SetbReady false)
      (if (not (Variables|Default|GetbScored))
        (Class|ShootingGalleryIF|TargetEnd (Variables|Default|GetGamePad) -5 (Variables|Default|GetMyBoardID)))
      (CallFunction|ResetAll))))

(event Custom|PlayerSteppedOff
  ;[1 0.3126 0] Server tracking active state of board
  (Utilities|FlowControl|DoOnce false
    (:Completed
      (Variables|Default|SetbAlive false)
      (Variables|Default|SetbReady false)
      (if (not (Variables|Default|GetbScored))
        (Class|ShootingGalleryIF|TargetEnd (Variables|Default|GetGamePad) -5 (Variables|Default|GetMyBoardID)))
      (CallFunction|ResetAll))))

(event Custom|OpenTarget
  ;[0 1 0.2211] Client rotating boards
  ;[0 0.9646 1] AUDIO - Boards Opening
  (Audio|PlaySoundatLocation "/Game/Athena/Sounds/Quests/Athena_ShootingGallery_TargetPop_Cue.Athena_ShootingGallery_TargetPop_Cue" (Transformation|GetActorLocation))
  (bind (_direction _float) (Timeline|FlipTimeline PlayFromStart)
    (:Update
      (Transformation|SetRelativeRotation (Variables|Components|GetSM_SimpleTargetPanel) (Math|Rotator|Lerp(Rotator) "0,0.000000,90.000000" "0, 0, 0" (Math|Float|Lerp 0.0 1.0 _float)) false false))))

(event Custom|CloseTarget
  ;[0 1 0.2211] Client rotating boards
  ;;[0 0.9646 1] AUDIO - Boards Closed
  (Audio|PlaySoundatLocation "/Game/Athena/Sounds/Quests/Athena_ShootingGallery_TargetDrop_Cue.Athena_ShootingGallery_TargetDrop_Cue" (Transformation|GetActorLocation))
  (bind (_direction _newtrack_0) (Timeline|CloseBoard ReverseFromEnd)
    (:Update
      (Transformation|SetRelativeRotation (Variables|Components|GetSM_SimpleTargetPanel) (Math|Rotator|Lerp(Rotator) "0,0.000000,90.000000" "0, 0, 0" (Math|Float|Lerp 1.0 0.0 _newtrack_0)) false false))))

(event EventDeactivateBoard
  ;[0.0502 0.6641 1] Shutdown
  (Utilities|FlowControl|DoOnce true
    (:Completed
      (Variables|Default|SetForceDeactivate true)
      (if (Variables|Default|GetbAlive)
        (Networking|SwitchHasAuthority
          (:Authority
            (Variables|Default|SetbReady false)
            (CallFunction|PlayerSteppedOff))
          (:Remote
            ;[0 1 0.02272] Client end
            (CallFunction|ResetAll)))))))

(event Custom|ClearDeactivate
  ;[0.0502 0.6641 1] Shutdown
  (Utilities|FlowControl|DoOnce true
    (:Completed
      (Variables|Default|SetForceDeactivate true)
      (if (Variables|Default|GetbAlive)
        (Networking|SwitchHasAuthority
          (:Authority
            (Variables|Default|SetbReady false)
            (CallFunction|PlayerSteppedOff))
          (:Remote
            ;[0 1 0.02272] Client end
            (CallFunction|ResetAll)))))))

(event EventBeginPlay
  ;[1 0.429 0.5698] Initialize
  (EventDispatchers|BindEventtoOnDamaged (delegate Custom|SampleDamage))
  (Transformation|SetRelativeRotation (Variables|Components|GetSM_SimpleTargetPanel) "0, 0,90.000000" false false))

(event Custom|SampleDamage
  (params (DamagedActor :type Actor%20Object%20Reference) (Damage :type Float%20%28single-precision%29) (InstigatedBy :type Controller%20Object%20Reference) (DamageCauser :type Actor%20Object%20Reference) (HitLocation :type Vector) (FHitComponent :type Primitive%20Component%20Object%20Reference) (BoneName :type Name) (Momentum :type Vector))
  ;[0.3115 0.7979 1] Take damage, check for validity
  (Gameplay|ForceBuildingHealth 1e+02)
  (Networking|SwitchHasAuthority
    (:Authority
      (if (Math|Boolean|ANDBoolean (Utilities|NotEqual(Object) DamageCauser "/Figment_S06/Weapons/FORT_Melee/Blueprints/Impact/B_Melee_Impact_Pickaxe_Athena.B_Melee_Impact_Pickaxe_Athena") (Utilities|Equal(Object) FHitComponent (Variables|Components|GetSM_SimpleTargetPanel)) (Variables|/Figment_S06/Athena/Items/QuestInteractables/ShootingGallery/Blueprint/BP_ShootingGallery_01.BP_ShootingGallery_01_C|GetbActive (Variables|Default|GetGamePad)) (Variables|Default|GetbReady))
        ;[1 0.3521 0.1511] Safety gating, make sure we don't overscore
        (Variables|Default|SetbReady false)
        (Variables|Default|SetbAlive false)
        (Utilities|FlowControl|DoOnce false
          (:Completed
            ;;[1 0.3126 0] Score!
            (Variables|Default|SetbScored true) ; Here is where the player has scored
            (CallFunction|ResetAll)))))
    (:Remote
      (if true
        ;;[0 0.9646 1] AUDIO - SHOT TARGET
        (Audio|PlaySoundatLocation "/Game/Athena/Sounds/Quests/Athena_ShootingGallery_HitTarget_Cue.Athena_ShootingGallery_HitTarget_Cue" (Transformation|GetActorLocation))))))

(event Custom|ResetAll
  ;[1 0.3521 0.1511] Safety gating, make sure we don't overscore
  (Utilities|FlowControl|DoOnce false
    (:Completed
      ;;[1 0.3126 0] Score!
      (Variables|Default|SetbScored true) ; Here is where the player has scored
      (CallFunction|ResetAll))))

(fn OnRep_bAlive
  (bind _closeboard (Variables|BP_ShootingTarget|GetCloseBoard))
  (bind _closeboard_2 (Variables|BP_ShootingTarget|GetCloseBoard))
  (bind _fliptimeline (Variables|BP_ShootingTarget|GetFlipTimeline))
  (if (Variables|Default|GetbAlive)
    ;[1 1 1] Board is alive and ready
    (Components|Timeline|SetNewTime _fliptimeline 0.0)
    (Components|Timeline|SetPlayRate _fliptimeline (Utilities|Array|Get(aref) (Variables|Default|GetPlayRate) (Variables|Default|GetCurrentRound)))
    (CallFunction|OpenTarget)
    (else
      ;[1 0.6843 0.1036] Board is done
      (if (Variables|Default|GetForceDeactivate)
        ;;[1 0.09225 0.142] Forced closed
        (Components|Timeline|SetNewTime _closeboard 0.0)
        (Components|Timeline|SetPlayRate _closeboard 5e+02)
        (CallFunction|CloseTarget)
        (Variables|Default|SetForceDeactivate false)
        (else
          (Components|Timeline|SetNewTime _closeboard_2 0.0)
          (Components|Timeline|SetPlayRate _closeboard_2 0.75)
          (CallFunction|CloseTarget))))))

(fn OnRep_BoardRotation)

(fn OnRep_bReady
  (Networking|SwitchHasAuthority
    (:Authority
      (if (Variables|Default|GetbReady)
        ;[1 0.3126 0] Server - Open boards
        (Transformation|SetRelativeRotation (Variables|Components|GetSM_SimpleTargetPanel) "0, 0, 0" false false)
        (else
          ;[1 0.3126 0] Server - Close boards
          (Transformation|SetRelativeRotation (Variables|Components|GetSM_SimpleTargetPanel) "0, 0,90.000000" false false))))))

(fn OnRep_bScored
  (bind _fliptimeline (Variables|BP_ShootingTarget|GetFlipTimeline))
  (if (Variables|Default|GetbScored)
    (Components|Timeline|SetNewTime _fliptimeline 5.0)
    (Components|Timeline|SetPlayRate _fliptimeline 3.0)
    (Class|ShootingGalleryIF|TargetEnd (Variables|Default|GetGamePad) 1 (Variables|Default|GetMyBoardID))
    (CallFunction|DeactivateBoard)))

(fn ConstructionScript)
