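; Blueprint interface for the shooting gallery quest interactable in Figment Season 6.
; It declares messages to begin a game and to activate a board for a given board ID and round, or deactivate it.
; TargetEnd passes a score, defaulting to one, and a board ID, likely reporting a finished target back to the game.
; It holds only function signatures, with no implementation.
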(fn ActivateBoard
  (params (BoardID :type Integer) (CurrentRound :type Integer)))

(fn BeginGame
  (return))

(fn DeactivateBoard)

(fn TargetEnd
  (params (Score :type Integer :default 1) (BoardID :type Integer)))
